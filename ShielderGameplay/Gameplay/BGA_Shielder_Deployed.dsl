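; Deployed Shielder shield actor: plays a deploy montage, runs on an energy component and lets friendly players retrieve it.
; Energy depletion, linked-actor loss or interaction starts the undeploy outro.
; Afterwards the server either spawns an item pickup and forces the instigator to collect it, or drops it, with a backup path for a missing controller.
; RepNotify flags hide mesh and collision; minigame play mode destroys shields left from edit mode.
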(event EventBeginPlay
  (bind _returnvalue (Game|GetInstigator))
  (bind _returnvalue_2 (Fortnite|GetFortPlayerControllerfromActor _returnvalue))
  (bind _returnvalue_3 (Actor|GetOwner))
  (bind _returnvalue_4 (Minigame|GetMinigame _returnvalue_3 :!class /Script/FortniteGame.MinigameBlueprintLibrary))
  (|CallParentFunction)
  (CallFunction|PlayDeployMontage)
  (bind _asfort_player_controller_athena (Utilities|Casting|CastToFortPlayerControllerAthena _returnvalue_2)
    (:then
      (bind _output_get (Variables|Default|SetInstigatorController _asfort_player_controller_athena))
      (Variables|/ShielderGameplay/Gameplay/B_Shield_EnergyTracker_Component.B_Shield_EnergyTracker_Component_C|SetInstigatorController _output_get (Variables|Default|GetShieldEnergyComponent)))
    (:CastFailed
      _))
  (Networking|SwitchHasAuthority
    (:Authority
      (Quest|SendComplexCustomStatEvent (Quests|GetQuestManager _returnvalue_2 "Invalid") _returnvalue_2 (Variables|Default|GetTagQuestDeployShield) "(GameplayTags=)" 1)
      ;[0.4937 0.4787 0.901 0.5] Ensure the deployed shield is destroyed when the minigame starts in case it was deployed in Edit Mode
      (Utilities|IsValid _returnvalue_4
        (:Is%20Valid
          (Minigame|AssignOnPlayModeChangedDelegate _returnvalue_4 (delegate Custom|OnPlayModeChangedDelegate_Event)))))))

(event Custom|OnPlayModeChangedDelegate_Event
  (params (Minigame :type Fort%20Minigame%20Object%20Reference) (bIsInPlayMode :type Boolean))
  ;[0.4937 0.4787 0.901 0.5] Ensure the deployed shield is destroyed when the minigame starts in case it was deployed in Edit Mode
  (if bIsInPlayMode
    (Variables|Default|SetIsDeactivated true)
    (Actor|DestroyActor)))

(event Collision|EventActorBeginOverlap
  (params (OtherActor :type Actor%20Object%20Reference)))

(event EventTick
  (params (DeltaSeconds :type Float%20%28single-precision%29)))

(event Energy|EventOnEnergyDepleted
  (if (not (Variables|Default|GetbUndeploying))
    (CallFunction|UndeployShield 0)))

(event Custom|UndeployShield
  (params (OverrideMontage :type Anim%20Montage%20Object%20Reference))
  (Variables|Default|SetbUndeploying true)
  (GameplayCue|ExecuteGameplayCue%28Burst%29Local self (Utilities|Struct|BreakGameplayCueTag (Variables|Default|GetRetractedCue)) (Ability|GameplayCue|MakeGameplayCueParameters 0.0))
  (CallFunction|DeployOutro OverrideMontage))

(event EventOnEnergyMaterialDeployAnimationFinished
  (params (WasPlayingForward :type Boolean))
  (if WasPlayingForward
    _))

(event EventOnDeployOutroFinished
  (Rendering|SetVisibility (Variables|Default|GetSkelMesh) false false)
  (Utilities|IsValid (Game|GetInstigatorController)
    (:Is%20Valid
      (CallFunction|ServerHandleUndeployFinished))
    (:Is%20Not%20Valid
      (CallFunction|ServerHandleUndeployFinishedBackup))) ; Account for there being no net owning client\n(e.g. they died)
)

(event Custom|PickupUndeploy
  (params (TargetLoc :type Vector))
  (CallFunction|UndeployShield (Variables|Default|GetPickupOutroMontage)))

(event Custom|PickupUndeployMoveSequence
  (params (TargetLoc :type Vector))
  (Components|Timeline|SetPlayRate (Variables|BGA_Shielder_Deployed|GetTL_MoveToPickupTarget) (Utilities|Reciprocal (Variables|Default|GetPickupUndeployMoveRate)))
  (Variables|Default|SetStartingLoc (Transformation|GetActorLocation))
  (bind (_direction _positionalpha) (Timeline|TL_MoveToPickupTarget PlayFromStart)
    (:Update
      (Transformation|SetActorLocation (Math|Vector|Lerp(Vector) (Variables|Default|GetStartingLoc) TargetLoc _positionalpha)))))

(event Interaction|EventBlueprintOnInteract
  (params (InteractingPawn :type Fort%20Pawn%20Object%20Reference) (InteractionBeingAttempted :type EInteractionBeingAttempted%20Enum :default "FirstInteraction"))
  (Variables|Default|SetbPickedUp true)
  (if (Variables|Default|GetIsDeactivated)
    (Actor|DestroyActor)
    (Variables|Default|SetHiddenAfterPickUp true)
    (else
      (CallFunction|PickupUndeploy (CallFunction|GetTargetLocation InteractingPawn))
      (Class|BShieldEnergyTrackerComponent|StopDepleting (Variables|Default|GetShieldEnergyComponent) false true))))

(event Custom|ServerHandleUndeployFinished
  (Utilities|FlowControl|DoOnce false
    (:Completed
      (if (Variables|Default|GetbPickedUp)
        (CallFunction|GiveShieldItem (Game|GetInstigator))
        (Variables|Default|SetHiddenAfterPickUp true)
        (Actor|DestroyActor)
        (else
          (Variables|Default|SetIsDeactivated true)
          (CallFunction|DropShieldItemPickup)
          (Actor|DestroyActor))))))

(event Custom|ServerHandleUndeployFinished_Backup
  (Networking|CanExecuteCosmeticEvents
    (:False
      (Utilities|FlowControl|DoOnce false
        (:Completed
          (if (Variables|Default|GetbPickedUp)
            (CallFunction|GiveShieldItem (Game|GetInstigator))
            (Variables|Default|SetHiddenAfterPickUp true)
            (Actor|DestroyActor)
            (else
              (Variables|Default|SetIsDeactivated true)
              (CallFunction|DropShieldItemPickup)
              (Actor|DestroyActor))))))) ; Ensure this is a dedicated server
)

(event OnLinkedActorDestroyed(FortLinkToActor)
  (params (DamageCauser :type Actor%20Object%20Reference))
  (CallFunction|OnLinkedActorInvalid))

(event OnLinkedPhysicsObjectWake(FortLinkToActor)
  (params (PhysicsObject :type Actor%20Object%20Reference))
  (CallFunction|OnLinkedActorInvalid))

(event OnLinkedActorDisconnected(FortLinkToActor)
  (CallFunction|OnLinkedActorInvalid))

(event Custom|OnLinkedActorInvalid
  (if (not (Variables|Default|GetbUndeploying))
    (CallFunction|UndeployShield 0)))

(event Custom|GiveShieldItem
  (params (Actor :type Actor%20Object%20Reference))
  (bind _returnvalue (Inventory|SpawnPickupInWorldWithClassAndItemEntry (Class|FortItemEntryComponent|GetOwnedItem (Variables|Default|GetFortItemEntry)) "/Script/FortniteGame.FortPickupAthena" (+ (Transformation|GetActorLocation) (Variables|Default|GetPickupSpawnOffset)) "0.000000,0.000000,1.000000" 0 true false false "Player" "TossedByPlayer" 0 false))
  (Utilities|FlowControl|Delay 0.04)
  (bind _asfort_player_pawn (Utilities|Casting|CastToFortPlayerPawn (Game|GetInstigator)))
  (Class|FortPlayerPawn|ServerHandlePickup _asfort_player_pawn _returnvalue 2.0 (Math|Vector|GetUnitDirection(Vector) (Transformation|GetActorLocation _returnvalue) (Transformation|GetActorLocation _asfort_player_pawn)) true))

(fn BlueprintCanInteract
  (params (InteractingPawn :type Fort%20Pawn%20Object%20Reference) (InteractionBeingAttempted :type EInteractionBeingAttempted%20Enum :default "FirstInteraction") (InteractionType :type TInteractionType%20Enum :default "IT_NoInteraction"))
  (return (and (Utilities|Enum|Equal(Enum) (Team|GetTeamAffiliation InteractingPawn (Variables|Default|GetInstigatorController)) "Friendly") (not (Variables|Default|GetbPickedUp)))))

(fn BlueprintDisplaySubInteractIcon
  (params (InteractingPawn :type Fort%20Pawn%20Object%20Reference) (InteractionBeingAttempted :type EInteractionBeingAttempted%20Enum :default "FirstInteraction") (InteractionType :type TInteractionType%20Enum :default "IT_NoInteraction"))
  (return false))

(fn BlueprintGetFailedSubTextTexture
  (params (InteractingPawn :type Fort%20Pawn%20Object%20Reference) (InteractionBeingAttempted :type EInteractionBeingAttempted%20Enum :default "FirstInteraction"))
  (return "/Game/Athena/HUD/Art/icon_shield.icon_shield"))

(fn BlueprintGetInteractionString
  (params (InteractingPawn :type Fort%20Pawn%20Object%20Reference) (InteractionBeingAttempted :type EInteractionBeingAttempted%20Enum :default "FirstInteraction"))
  (bind _result (ArrayMacros|SafeArrayGet (Variables|Default|GetRarityTexts) (Math|Conversions|ToInteger(Byte) (Item|GetRarity (Variables|Default|GetShieldItemDefinition))))
    (:%20%20
      (return (Utilities|Text|FormatText "NSLOCTEXT(\"[17FAA8374DC1F90FE8708A94372D0E24]\", \"D63BA26049E52F99842DA09533DECCB9\", \"Retrieve {DisplayName}\\r\\n{Rarity}\")" (ItemDefinition|GetDisplayName (Variables|Default|GetShieldItemDefinition)) _result)))))

(fn BlueprintGetSubInteractionString
  (params (InteractingPawn :type Fort%20Pawn%20Object%20Reference) (InteractionBeingAttempted :type EInteractionBeingAttempted%20Enum :default "FirstInteraction"))
  (return (Utilities|Text|FormatText "NSLOCTEXT(\"[17FAA8374DC1F90FE8708A94372D0E24]\", \"A66A71984CFD50E816A805A4DBAC47DD\", \"Energy: {ChargePercent}%\")" (Math|Float|Ceil (* (Variables|/ShielderGameplay/Gameplay/B_Shield_EnergyTracker_Component.B_Shield_EnergyTracker_Component_C|GetCurrentEnergy (Variables|Default|GetShieldEnergyComponent)) 1e+02)))))

(fn DropShieldItemPickup
  (Inventory|SpawnPickupInWorldWithClassAndItemEntry (Class|FortItemEntryComponent|GetOwnedItem (Variables|Default|GetFortItemEntry)) "/Script/FortniteGame.FortPickupAthena" (+ (Transformation|GetActorLocation) (Variables|Default|GetPickupSpawnOffset))))

(fn ForcePickup)

(fn GetCustomIcon
  (return ""))

(fn GetShieldItemPickupLocation
  (return (+ (Transformation|GetActorLocation) (Variables|Default|GetPickupSpawnOffset))))

(fn GetTargetLocation
  (params (TargetActor :type Actor%20Object%20Reference))
  (return (+ (Transformation|GetActorLocation TargetActor) (Math|Vector|RotateVector (Variables|Default|GetTargetRelativeOffset) (Transformation|GetActorRotation TargetActor)))))

(fn OnRep_HiddenAfterPickUp
  (Rendering|SetVisibility (Variables|Default|GetSkelMesh) (not (Variables|Default|GetHiddenAfterPickUp)) false)
  (Collision|SetCollisionEnabled (Variables|Default|GetCollisionMesh) "NoCollision"))

(fn OnRep_IsDeactivated
  (if (Variables|Default|GetIsDeactivated)
    (CallFunction|SetDeactivatedState)))

(fn SetDeactivatedState
  (Collision|SetCollisionEnabled (Variables|Default|GetCollisionMesh) "NoCollision")
  (Rendering|SetVisibility (Variables|Default|GetSkelMesh) false false)
  (Collision|SetCollisionEnabled (Variables|Default|GetSmallInteractionCollision) "NoCollision"))

(fn ConstructionScript
  (|CallParentFunction))
