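; Held Shielder energy shield actor that switches the player's shield and weapon visibility and transform by replicated shield mode.
; Mode changes look up properties in a map and trigger deploy/retract animations, with server-played cues.
; Deferred transitions hold properties until the outro finishes, using a server RPC to work around missed montage callbacks.
; The authority spawns a separate collision actor on the instigator's capsule, toggled with shield visibility and destroyed alongside it.
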(event EventBeginPlay
  (|CallParentFunction)
  (Actor|SetOwner (Game|GetInstigator))
  (Networking|SwitchHasAuthority
    (:Authority
      (CallFunction|SpawnCollisionActor))))

(event EventOnEnergyMaterialDeployAnimationFinished
  (params (WasPlayingForward :type Boolean)))

(event Energy|EventOnEnergyDepleted
  (Variables|Default|SetbIsRecharging true))

(event EventOnEnergyFull
  (Variables|Default|SetbIsRecharging false))

(event EventOnDeployOutroFinished
  (if (Variables|Default|GetbHasPendingPropertiesToApply)
    (CallFunction|ApplyPendingProperties)
    (CallFunction|ServerApplyPendingProperties)))

(event Custom|Server_ApplyPendingProperties
  ;[1 1 1] Work around for server for some reason not receiving the montage finished callback that drives applying pending properties
  (CallFunction|ApplyPendingProperties))

(event EventDestroyed
  (bind _collisionactor (Variables|Default|GetCollisionActor))
  (|CallParentFunction)
  (Utilities|IsValid _collisionactor
    (:Is%20Valid
      (Actor|DestroyActor _collisionactor))))

(fn ApplyPendingProperties
  (CallFunction|ApplyShieldProperties (Variables|Default|GetPendingProps))
  (Variables|Default|SetbHasPendingPropertiesToApply false))

(fn ApplyShieldProperties
  (params (Props :type FShield%20Weapon%20Properties%20Structure))
  (CallFunction|SetTransformfromShieldProperties Props)
  (CallFunction|SetMeshesfromShieldProperties Props)
  (CallFunction|SetVisibilityfromShieldProperties Props))

(fn CheckShouldPlayTransitionCues
  (if (Cues|ShouldPlayinCue)
    (Cues|PlayinCue)
    (CallFunction|PlayDeployMontage)
    (elif (Cues|ShouldPlayOutCue)
      (Cues|PlayOutCue)
      (CallFunction|DeployOutro 0))))

(fn HandleNewMode
  (params (NewMode :type E_ShieldWeaponMode%20Enum :default "NewEnumerator0"))
  (bind _modeproperties (Variables|Default|GetModeProperties))
  (bind (_value _returnvalue) (Utilities|Map|Find _modeproperties NewMode))
  (LocalVariables|SetCurrentProps _value)
  (if _returnvalue
    ;[1 1 1] Certain mode transitions should play GCs and animations, so kick those off here
    (Cues|CheckShouldPlayTransitionCues)
    ;[1 1 1] "Exit" modes (e.g.shielding->unequipped) might make one of the meshes invisible, so we save the properties to apply later once all animations are finished
    ;
    ; However an animation doesn't play when setting the initial unequipped mode, so we make an exception for the first time this function is run
    (if (and (CallFunction|IsDeferredTransition (Variables|Default|GetPrevMode) (Variables|GetNewMode)) (not (Variables|Default|GetbIsInitialMode)))
      (CallFunction|SetPendingProperties (LocalVariables|GetCurrentProps))
      (else
        (CallFunction|ApplyShieldProperties (LocalVariables|GetCurrentProps))
        (if (Variables|Default|GetbIsInitialMode)
          (Variables|Default|SetbIsInitialMode false))))))

(fn IsDeferredTransition
  (params (OldMode :type E_ShieldWeaponMode%20Enum :default "NewEnumerator0") (NewMode :type E_ShieldWeaponMode%20Enum :default "NewEnumerator0"))
  (bind _array_element (Utilities|Array|ForEachLoopwithBreak (Variables|Default|GetDeferredTransitions))
    (:LoopBody
      (if (CallFunction|TransitionMatches OldMode NewMode _array_element)
        (return true)))
    (:Completed
      (return false))))

(fn OnRep_CollisionActor
  (bind _collisionactor (Variables|Default|GetCollisionActor))
  (Utilities|IsValid _collisionactor
    (:Is%20Valid
      (Class|BGAShielderCollisionHeld|InitfromParent _collisionactor self))))

(fn OnRep_CurrentMode
  (bind _currentmode (Variables|Default|GetCurrentMode))
  (Development|PrintString (Utilities|String|Append "Replicating new mode: " (Utilities|String|EnumtoString _currentmode)) true true "(R=1.000000,G=0.000000,B=0.089921,A=1.000000)")
  (Mode|HandleNewMode _currentmode))

(fn PlayInCue
  (CallFunction|PlayEnergyDeployTimeline true)
  (Networking|SwitchHasAuthority
    (:Authority
      (Cues|PlayCueonInstigator (Variables|Default|GetDeployedCue)))))

(fn PlayOutCue
  (CallFunction|PlayEnergyDeployTimeline false)
  (Networking|SwitchHasAuthority
    (:Authority
      (Cues|PlayCueonInstigator (Variables|Default|GetRetractCue)))))

(fn SetMeshesFromShieldProperties
  (params (Props :type FShield%20Weapon%20Properties%20Structure)))

(fn SetPendingProperties
  (params (Props :type FShield%20Weapon%20Properties%20Structure))
  (Variables|Default|SetPendingProps Props)
  (Variables|Default|SetbHasPendingPropertiesToApply true))

(fn SetShieldMode
  (params (Mode :type E_ShieldWeaponMode%20Enum :default "NewEnumerator0"))
  (if (Utilities|Enum|NotEqual(Enum) Mode (Variables|Default|GetCurrentMode))
    (if (not (Utilities|Enum|Equal(Enum) (Variables|Default|GetCurrentMode) "NewEnumerator6"))
      (Variables|Default|SetPrevMode (Variables|Default|GetCurrentMode)))
    (Variables|Default|SetCurrentMode (Variables|GetMode))))

(fn SetShieldVisible
  (params (Visible :type Boolean))
  (bind _collisionactor (Variables|Default|GetCollisionActor))
  (Rendering|SetVisibility (Variables|Default|GetSkelMesh) Visible true)
  (Development|PrintString (Utilities|String|Append "Visibility: " (Utilities|String|ToString(Boolean) Visible)))
  (Utilities|IsValid _collisionactor
    (:Is%20Valid
      (Class|BGAShielderCollisionHeld|EnableCollision _collisionactor (Variables|GetVisible)))))

(fn SetTransformFromShieldProperties
  (params (Props :type FShield%20Weapon%20Properties%20Structure))
  (bind (_attachsocket_10_9cc116324d6b0d331a36858dd2571915 _energymesh_16_28d5b7fb4ddffbc96420138e00203d64 _shieldmesh_18_c782f7fc4e90fa3b8c2201a058a5d4fb _shieldrelativetransform_32_edb888dc4ad1103014c418a84a2d970e) (Utilities|Struct|BreakFShieldWeaponProperties Props))
  (Transformation|SetRelativeTransform (Variables|Default|GetSkelMesh) _shieldrelativetransform_32_edb888dc4ad1103014c418a84a2d970e false false))

(fn SetVisibilityFromShieldProperties
  (params (Props :type FShield%20Weapon%20Properties%20Structure))
  (bind _shieldmeshvisible_38_512be53b485d8c715038f6b17f69578b (Utilities|Struct|BreakFShieldWeaponProperties Props))
  (Mode|SetShieldVisible _shieldmeshvisible_38_512be53b485d8c715038f6b17f69578b)
  (Class|BAthenaShieldWeapon|SetWeaponActorVisibility (Variables|Default|GetWeaponActorRef) (not _shieldmeshvisible_38_512be53b485d8c715038f6b17f69578b)))

(fn ShouldPlayInCue
  (return (and (or (Utilities|Enum|Equal(Enum) (Variables|Default|GetCurrentMode) "NewEnumerator2") (Utilities|Enum|Equal(Enum) (Variables|Default|GetCurrentMode) "NewEnumerator5")) (or (Utilities|Enum|Equal(Enum) (Variables|Default|GetPrevMode) "NewEnumerator0") (Utilities|Enum|Equal(Enum) (Variables|Default|GetPrevMode) "NewEnumerator3")))))

(fn ShouldPlayOutCue
  (return (and (or (Utilities|Enum|Equal(Enum) (Variables|Default|GetPrevMode) "NewEnumerator2") (Utilities|Enum|Equal(Enum) (Variables|Default|GetPrevMode) "NewEnumerator5")) (or (Utilities|Enum|Equal(Enum) (Variables|Default|GetCurrentMode) "NewEnumerator0") (Utilities|Enum|Equal(Enum) (Variables|Default|GetCurrentMode) "NewEnumerator3")))))

(fn SpawnCollisionActor
  (bind _returnvalue (Building|SpawnBuildingGameplayActor (Variables|Default|GetCollisionActorClass) (Transformation|GetActorTransform) (Variables|Actor|GetInstigator) true))
  (bind _asbga_shielder_collision_held (Utilities|Casting|CastToBGA_Shielder_Collision_Held _returnvalue))
  (bind _output_get (Variables|Default|SetCollisionActor _asbga_shielder_collision_held))
  (Class|BGAShielderCollisionHeld|InitfromParent _output_get self)
  (bind _asfort_player_pawn (Utilities|Casting|CastToFortPlayerPawn (Variables|Actor|GetInstigator)))
  (Transformation|AttachActorToComponent _output_get (Variables|/Script/Engine.Character|GetCapsuleComponent _asfort_player_pawn) "None" "SnapToTarget" "SnapToTarget" "KeepWorld" false)
  (Transformation|SetActorRelativeTransform _output_get (Variables|Default|GetCollisionActorInitialRelativeTransform) false false))

(fn ConstructionScript)
