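; Base Blueprint for the Shielder's shield actor, which never dies from damage and spends energy instead.
; Server damage events drain the energy component, fire a hit cue, play scaled material ripples and impact FX, and batch blocked-hit quest reports on a timer.
; Deploy and outro montages swap materials and toggle Niagara and audio.
; The CurrentEnergy rep notify drives low-energy flashing and warning sound, while several energy and transition hooks are empty.
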(event Custom|PlayDamageRipple
  (params (Damage :type Float%20%28double-precision%29) (Position :type Vector) (Normal :type Vector) (Reflected :type Vector))
  (bind _returnvalue (Math|Float|NormalizetoRange Damage 6.0 1.2e+02))
  (bind _returnvalue_2 (Math|Float|Clamp(Float) _returnvalue))
  (bind _skelmesh (Variables|Default|GetSkelMesh))
  ;[1 1 1] Hit material animation and FX
  (Networking|CanExecuteCosmeticEvents
    (:True
      (Utilities|FlowControl|Sequence
        (:then_0
          (Rendering|Material|SetVectorParameterValueonMaterials _skelmesh "RipplePosition" Position)
          (Rendering|Material|SetScalarParameterValueonMaterials _skelmesh "RippleRadius" (Math|Float|Lerp 9e+01 5.5e+02 _returnvalue))
          (Rendering|Material|SetScalarParameterValueonMaterials _skelmesh "RippleDuration" (Math|Float|Lerp 0.18 1.0 _returnvalue_2))
          (Rendering|Material|SetScalarParameterValueonMaterials _skelmesh "RippleTime" (Utilities|Time|GetGameTimeinSeconds)))
        (:then_1
          (if (> Damage 15.0)
            (bind _returnvalue_3 (Effects|Components|ParticleSystem|SpawnFXAtLocation "/ShielderGameplay/FX/NS_ShieldWeapon_DamageImpact.NS_ShieldWeapon_DamageImpact" Position (Math|Conversions|RotationFromXVector Normal)))
            (Effects|Components|ParticleSystem|SetVectorParameter _returnvalue_3 "Ricochet" Reflected)
            (Effects|Components|ParticleSystem|SetFloatParameter _returnvalue_3 "DamageNormalized" _returnvalue_2)))))))

(event EventBeginPlay
  (Energy|InitializeEnergyListeners)
  (Effects|Components|ParticleSystem|SetBoolParameter (Variables|Default|GetNS_ShieldWeapon_Active) "Big" (Variables|Default|GetBigShield)) ; Adapt FX for big version
  (Rendering|Material|SetMaterial (Variables|Default|GetSkelMesh) 2 (Variables|Default|GetEnergyMaterialDeploying)))

(event EventOnDamageServer
  (params (Damage :type Float%20%28single-precision%29) (DamageTags :type Gameplay%20Tag%20Container%20Structure%20%28by%20ref%29) (Momentum :type Vector :default "0, 0, 0") (HitInfo :type Hit%20Result%20Structure%20%28by%20ref%29) (InstigatedBy :type Controller%20Object%20Reference) (DamageCauser :type Actor%20Object%20Reference) (EffectContext :type Gameplay%20Effect%20Context%20Handle%20Structure))
  (bind (_bblockinghit _binitialoverlap _time _distance _location _impactpoint _normal _impactnormal _physmat _hitactor _hitcomponent _hitbonename _bonename _hititem _elementindex _faceindex _tracestart _traceend) (Collision|BreakHitResult HitInfo))
  (Ability|SendGameplayEventtoActor InstigatedBy "(TagName=\"Event.Shield.Hit\")" (Utilities|Struct|MakeGameplayEventData "(TagName=\"\")" InstigatedBy self 0 0 "()" "(GameplayTags=)" "(GameplayTags=((TagName=\"TargetFilter.AttachedActorAlwaysObstructsWorld\")))"))
  (Damage|HandleDamage Damage HitInfo)
  (CallFunction|PlayDamageRipple Damage _location _normal (Math|Vector|GetReflectionVector (Math|Vector|Normalize (- _traceend _tracestart)) _normal)))

(event Custom|PlayEnergyDeployTimeline
  (params (Forward :type Boolean)))

(event Custom|PlayDeployMontage
  (bind _opensound (Variables|Audio|GetOpenSound))
  (bind _skelmesh (Variables|Default|GetSkelMesh))
  (bind _activeloop (Variables|Default|GetActiveLoop))
  (Variables|Default|SetbDeployOutroFinished false)
  (bind _notifyname (Animation|Montage|PlayMontage (Variables|Default|GetSkelMesh) (Variables|Default|GetDeployMontage))
    (:then
      (CallFunction|SetShieldEnergyMaterial true)
      (Rendering|Material|SetMaterial (Variables|Default|GetSkelMesh) 0 "/ShielderGameplay/Materials/M_ShieldWeapon_Deployed.M_ShieldWeapon_Deployed")
      (Components|Activation|SetActive (Variables|Default|GetNS_ShieldWeapon_Active) true false)
      (Audio|SpawnSoundAttached _opensound (CallFunction|GetAttachmentComponent))
      (SoundIndicator|ShowSoundIndicator (Variables|Default|GetFortSoundIndicator) (Transformation|GetActorLocation) _opensound 1.0 0)
      (Audio|Components|Audio|FadeIn _activeloop (Variables|Audio|GetActiveLoopFadeInTime) 1.0 0.0 "Linear" :!class /Script/Engine.AudioComponent)
      (Audio|SetAudioAttached _activeloop true)
      (bind _returnvalue (Effects|SpawnFXSystemAttached :FXSystemAsset (Variables|Default|GetDeployFX) :AttachToComponent (Variables|Default|GetSkelMesh) :AttachPointName "shield_base" :Scale "1.000000,1.000000,1.000000" :LocationType "SnapToTarget" :bAutoDestroy true :PoolingMethod "None" :bAutoActivate true))
      (bind _asfort_pawn (Utilities|Casting|CastToFortPawn (Game|GetInstigator)))
      (Niagara|SetNiagaraSkeletalMeshComponent (Variables|Default|GetNS_ShieldWeapon_Active) "PlayerMesh" (Variables|/Script/Engine.Character|GetMesh _asfort_pawn)))
    (:OnNotifyBegin
      (if (== _notifyname "Deployed")
        (Culling|UpdateCullDistance false)
        (CallFunction|OnEnergyMaterialDeployAnimationFinished true)
        (Networking|CanExecuteCosmeticEvents
          (:True
            (Variables|Default|SetbDeployed true)
            (CallFunction|SetShieldEnergyMaterial false)
            (Rendering|Material|SetVectorParameterValueonMaterials _skelmesh "RipplePosition" (Transformation|GetSocketLocation (Variables|Default|GetSkelMesh) "shield_base"))
            (Rendering|Material|SetScalarParameterValueonMaterials _skelmesh "RippleRadius" (Math|Float|SelectFloat 7.5e+02 4e+02 (Variables|Default|GetBigShield)))
            (Rendering|Material|SetScalarParameterValueonMaterials _skelmesh "RippleDuration" 0.3)
            (Rendering|Material|SetScalarParameterValueonMaterials _skelmesh "RippleTime" (Utilities|Time|GetGameTimeinSeconds))))))))

(event Custom|DeployOutro
  (params (OverrideMontage :type Anim%20Montage%20Object%20Reference))
  (bind _closesound (Variables|Audio|GetCloseSound))
  (bind _activeloop (Variables|Default|GetActiveLoop))
  (Networking|CanExecuteCosmeticEvents
    (:True
      (Animation|Montage|PlayMontage (Variables|Default|GetSkelMesh) (select (Utilities|IsValid OverrideMontage :!class /Script/Engine.KismetSystemLibrary) OverrideMontage (Variables|Default|GetDeployMontage)) 1.0 0.0 "Outro"
        (:then
          (Variables|Default|SetbDeployed false)
          (CallFunction|SetShieldEnergyMaterial true)
          (Rendering|Material|SetMaterial (Variables|Default|GetSkelMesh) 0 "/ShielderGameplay/Materials/MI_Shielder_Handheld.MI_Shielder_Handheld")
          (Components|Activation|SetActive (Variables|Default|GetNS_ShieldWeapon_Active) false false)
          (Audio|SpawnSoundAttached _closesound (CallFunction|GetAttachmentComponent))
          (SoundIndicator|ShowSoundIndicator (Variables|Default|GetFortSoundIndicator) (Transformation|GetActorLocation) _closesound 1.0 0)
          (Audio|Components|Audio|FadeOut _activeloop (Variables|Audio|GetActiveLoopFadeOutTime) 0.0 "Linear" :!class /Script/Engine.AudioComponent)
          (Audio|SetAudioAttached _activeloop)
          (Audio|Components|Audio|FadeOut (Variables|Default|GetLowEnergyAudioComp) 0.0 0.0 "Linear" :!class /Script/Engine.AudioComponent)
          (Utilities|IsValid (stale Custom|PlayDeployMontage _returnvalue)
            (:Is%20Valid
              (Components|Activation|SetActive (stale Custom|PlayDeployMontage _returnvalue) false false) ; Kill FX if we interrupt the deploy
            )))
        (:OnCompleted
          _)
        (:OnBlendOut
          _)
        (:OnInterrupted
          _)))
    (:False
      (Utilities|IsValid (Game|GetInstigatorController)
        (:Is%20Not%20Valid
          _)) ; If on server with no current net owning client, go ahead and finish it now. Otherwise it'll never happen.
    )) ; Check if NOT a dedicated server
  (CallFunction|OnEnergyMaterialDeployAnimationFinished false)
  (CallFunction|OnDeployOutroFinished)
  (CallFunction|SetShieldEnergyMaterial false))

(event Custom|OnDeployOutroFinished
  (Audio|Components|Audio|FadeOut (Variables|Default|GetActiveLoop) 0.5 0.0 "Linear" :!class /Script/Engine.AudioComponent))

(event Custom|ReportHitsForQuest
  (Quest|SendComplexCustomStatEvent (Quests|GetQuestManager (Fortnite|GetFortPlayerControllerfromActor (Game|GetInstigator)) "Invalid") (Game|GetInstigator) (Variables|Default|GetTagQuestBlockedHits) "(GameplayTags=)" (Variables|Default|GetNumHitsSinceLastReport))
  (Variables|Default|SetNumHitsSinceLastReport 0)
  (Utilities|Time|ClearandInvalidateTimerbyHandle (Variables|Default|GetQuestHitReportTimer)))

(fn CalculateEnergyCostfromDamage
  (params (Damage :type Float%20%28double-precision%29))
  (Development|PrintString (Utilities|String|ToString(Float) Damage))
  (return (/ Damage (ScalableFloat|GetValueatLevel (Variables|Default|GetRow_MaxDamageToDeplete)))))

(fn DepleteEnergyFromDamage
  (params (Damage :type Float%20%28double-precision%29))
  (Class|BShieldEnergyTrackerComponent|ModifyEnergy (Variables|Default|GetShieldEnergyComponent) (* (Damage|CalculateEnergyCostFromDamage Damage) -1.0)))

(fn GetAttachmentComponent
  (return (select (Variables|Audio|GetbAttachAudioToPlayer) (Variables|Transformation|GetRootComponent (Game|GetInstigator)) (Variables|Default|GetSkelMesh))))

(fn GetShieldColorByAlpha
  (params (Alpha :type Float%20%28double-precision%29))
  (return (Math|Curves|GetLinearColorValue (Variables|Material|GetEnergyColorCurve) (- 1.0 Alpha))))

(fn HandleDamage
  (params (Damage :type Float%20%28double-precision%29) (HitResult :type Hit%20Result%20Structure) (DamageCauser :type Actor%20Object%20Reference))
  (bind _returnvalue (Utilities|Time|GetGameTimeinSeconds))
  (Energy|DepleteEnergyfromDamage Damage)
  (Ability|ExecuteGameplayCueWithParamsReplicated (Variables|Cues|GetDamageCue_Energy) (GameplayCue|MakeGameplayCueParametersfromHitResult HitResult))
  (if (>= (- _returnvalue (Variables|Default|GetTimeLastHit)) (Variables|Default|GetMinTimeBetweenHits))
    (Variables|Default|SetTimeLastHit _returnvalue)
    ;[1 1 1] Batch reporting the number of hits blocked for quests. If we are hit and there isn't a pending timer, we will start one, and all of the hits received before the timer finishes will be reported
    (Math|Integer|IncrementInt (Variables|Default|GetNumHitsSinceLastReport)
      (:%20%20
        (if (not (Utilities|Time|IsValidTimerHandle (Variables|Default|GetQuestHitReportTimer)))
          (bind _returnvalue_2 (Utilities|Time|SetTimerbyEvent (EventDispatchers|CreateEvent :!event ReportHitsForQuest) 1.0))
          (Variables|Default|SetQuestHitReportTimer _returnvalue_2))))))

(fn InitializeEnergyListeners
  (bind _shieldenergycomponent (Variables|Default|GetShieldEnergyComponent))
  (Default|BindEventtoOnEnergyDepleted _shieldenergycomponent (EventDispatchers|CreateEvent :!event OnEnergyDepleted))
  (Default|BindEventtoOnEnergyUpdated _shieldenergycomponent (EventDispatchers|CreateEvent :!event OnEnergyUpdated))
  (Default|BindEventtoOnEnergyFull _shieldenergycomponent (EventDispatchers|CreateEvent :!event OnEnergyFull)))

(fn InitializeEnergyMaterialTimeline
  (Components|Timeline|SetPlayRate (Variables|BGA_Shielder_Parent|GetTL_EnergyMaterialDeploy) (Utilities|Reciprocal (Variables|Material|GetEnergyMaterialAnimDuration))))

(fn InitializeMaterial
  (bind _skelmesh (Variables|Default|GetSkelMesh))
  (bind _energymatindex (Variables|Default|GetEnergyMatIndex))
  (bind _returnvalue (Rendering|Material|CreateDynamicMaterialInstance _skelmesh _energymatindex 0 "None" :!class /Script/Engine.PrimitiveComponent))
  (Rendering|Material|SetMaterial _skelmesh _energymatindex _returnvalue)
  (Variables|Material|SetEnergyMat _returnvalue))

(fn OnEnergyDepleted)

(fn OnEnergyFull)

(fn OnEnergyMaterialDeployAnimationFinished
  (params (WasPlayingForward :type Boolean)))

(fn OnEnergyUpdated
  (params (Energy :type Float%20%28double-precision%29))
  (Variables|Default|SetCurrentEnergy Energy))

(fn OnRep_bDeployOutroFinished
  (if (Variables|Default|GetbDeployOutroFinished)
    (CallFunction|OnDeployOutroFinished)))

(fn OnRep_bPlayDeployMontage)

(fn OnRep_CurrentEnergy
  (bind _flashingthreshold (Variables|Default|GetFlashingThreshold))
  (bind _currentenergy (Variables|Default|GetCurrentEnergy))
  (bind _returnvalue (< _currentenergy _flashingthreshold))
  (bind _returnvalue_2 (Math|Float|SelectFloat 1.0 0.0 _returnvalue))
  (Rendering|Material|SetScalarParameterValueonMaterials (Variables|Default|GetSkelMesh) "Flashing" _returnvalue_2)
  (Effects|Components|ParticleSystem|SetFloatParameter (Variables|Default|GetNS_ShieldWeapon_Active) "Flashing" _returnvalue_2)
  (if (and _returnvalue (Variables|Default|GetbDeployed))
    (Variables|Default|GetLowEnergyAudioComp
      (:then)
      (:else
        (bind _returnvalue_3 (Audio|SpawnSoundAttached (Variables|Audio|GetLowEnergyLoopSound) (CallFunction|GetAttachmentComponent)))
        (bind _output_get (Variables|Default|SetLowEnergyAudioComp _returnvalue_3))
        (Audio|Components|Audio|FadeIn _output_get (Variables|Audio|GetLowEnergyFadeInTime) 1.0 0.0 "Linear" :!class /Script/Engine.AudioComponent)))
    (else
      (bind _lowenergyaudiocomp (Variables|Default|GetLowEnergyAudioComp)
        (:then
          (Audio|Components|Audio|FadeOut _lowenergyaudiocomp (Variables|Audio|GetLowEnergyFadeOutTime) 0.0 "Linear" :!class /Script/Engine.AudioComponent))
        (:else)))))

(fn PlayCueOnInstigator
  (params (GameplayCueTag :type Gameplay%20Tag%20Structure)))

(fn PlayDamageCues
  (params (HitResult :type Hit%20Result%20Structure) (Damage :type Float%20%28double-precision%29))
  (Ability|ExecuteGameplayCueWithParamsReplicated (Variables|Cues|GetDamageCue_Energy) (GameplayCue|MakeGameplayCueParametersfromHitResult HitResult)))

(fn SetAudioAttached
  (params (Audio :type Scene%20Component%20Object%20Reference) (Attach :type Boolean))
  (if Attach
    (Transformation|AttachComponentToComponent Audio (CallFunction|GetAttachmentComponent) "None" "SnapToTarget" "SnapToTarget" "SnapToTarget" true)
    (else
      (Transformation|DetachFromComponent Audio "KeepWorld" "KeepWorld" "KeepWorld" true))))

(fn SetEnergyMaterialColor
  (params (CenterColor :type Linear%20Color%20Structure))
  (Rendering|Material|SetVectorParameterValue (Variables|Material|GetEnergyMat) "ColorCenter" CenterColor :!class /Script/Engine.MaterialInstanceDynamic))

(fn SetEnergyMaterialDeployParameter
  (params (Value :type Float%20%28double-precision%29))
  (Rendering|Material|SetScalarParameterValue (Variables|Material|GetEnergyMat) "Deploy" Value :!class /Script/Engine.MaterialInstanceDynamic))

(fn SetEnergyMatRippleParams
  (params (HitLoc :type Vector) (GridSize :type Float%20%28double-precision%29) (GridRange :type Float%20%28double-precision%29) (OuterRingSize :type Float%20%28single-precision%29) (InnerRingSize :type Float%20%28single-precision%29) (Duration :type Float%20%28single-precision%29) (GridBrightness :type Float%20%28single-precision%29))
  (bind _energymat (Variables|Material|GetEnergyMat))
  (Rendering|Material|SetVectorParameterValue _energymat "HitWave-Loc" (Math|Conversions|ToLinearColor(Vector) HitLoc) :!class /Script/Engine.MaterialInstanceDynamic)
  (Rendering|Material|SetScalarParameterValue _energymat "GameTime" (Utilities|Time|GetGameTimeinSeconds) :!class /Script/Engine.MaterialInstanceDynamic)
  (Rendering|Material|SetScalarParameterValue _energymat "GridSize" (Variables|GetGridSize) :!class /Script/Engine.MaterialInstanceDynamic)
  (Rendering|Material|SetScalarParameterValue _energymat "GridRange" (Variables|GetGridRange) :!class /Script/Engine.MaterialInstanceDynamic)
  (Rendering|Material|SetScalarParameterValue _energymat "HitWave-SizeMultiplierOuterRing" (Variables|GetOuterRingSize) :!class /Script/Engine.MaterialInstanceDynamic)
  (Rendering|Material|SetScalarParameterValue _energymat "HitWave-SizeMultiplierInnerRing" (Variables|GetInnerRingSize) :!class /Script/Engine.MaterialInstanceDynamic)
  (Rendering|Material|SetScalarParameterValue _energymat "HitWave-DurationNew" (Variables|GetDuration) :!class /Script/Engine.MaterialInstanceDynamic)
  (Rendering|Material|SetScalarParameterValue _energymat "Hitwave_WPO_Direction" 0.0 :!class /Script/Engine.MaterialInstanceDynamic)
  (Rendering|Material|SetScalarParameterValue _energymat "HitEffectLerp" (Variables|GetGridBrightness) :!class /Script/Engine.MaterialInstanceDynamic))

(fn SetShieldEnergyMaterial
  (params (Deploying :type Boolean))
  (Rendering|Material|SetMaterial (Variables|Default|GetSkelMesh) 2 (select Deploying (Variables|Default|GetEnergyMaterialDeploying) (Variables|Default|GetEnergyMaterialDeployed))))

(fn ShouldDie
  (params (Damage :type Float%20%28single-precision%29) (EventInstigator :type Controller%20Object%20Reference) (DamageCauser :type Actor%20Object%20Reference) (bIsTest :type Boolean) (EffectContext :type Gameplay%20Effect%20Context%20Handle%20Structure%20%28by%20ref%29))
  (return false))

(fn TransitionMatches
  (params (OldMode :type E_ShieldWeaponMode%20Enum :default "NewEnumerator0") (NewMode :type E_ShieldWeaponMode%20Enum :default "NewEnumerator0") (Transition :type FShield%20Mode%20Transition%20Structure))
  (bind (_frommode_2_6263cf9b494a8bd38952c1a5c03d67e2 _tomode_4_cfa72583472cbc3539e23fae4717ed9f) (Utilities|Struct|BreakFShieldModeTransition Transition))
  (return (and (Utilities|Enum|Equal(Enum) OldMode _frommode_2_6263cf9b494a8bd38952c1a5c03d67e2) (Utilities|Enum|Equal(Enum) NewMode _tomode_4_cfa72583472cbc3539e23fae4717ed9f))))

(fn ConstructionScript)
