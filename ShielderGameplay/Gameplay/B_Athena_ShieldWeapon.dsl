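; Fortnite Shielder weapon that gives its wielder a held, energy-based shield actor.
; On authority it spawns the shield BGA on the pawn's mesh, initializes its energy after the first ammo change, and destroys it with the weapon.
; Gameplay-tag listeners track shield activation, vehicle entry and throwing, while touch input toggles shielding through gameplay events.
; RepNotifies drive animation overrides, shield mode and energy widget fill; several handlers, including ModifyReticle, are empty stubs.
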(event EventBeginPlay
  ;[1 1 1] Initialization
  (|CallParentFunction)
  (CallFunction|ListentoActorEvents)
  (CallFunction|InitializeActionPrompts)
  (CallFunction|ShieldFormTagListeners)
  (CallFunction|VehicleTagListeners)
  (CallFunction|StartThrowTagListeners)
  (Game|BindEventtoOnDestroyed (EventDispatchers|CreateEvent :!event Custom|OnShieldWeaponDestroyed))
  (if (TouchPlatform|IsUsingTouch (CallFunction|GetPlayerController))
    (EventDispatchers|BindEventtoOnPressTrigger (EventDispatchers|CreateEvent :!event Custom|TouchPressTrigger%20)))
  (Networking|SwitchHasAuthority
    (:Authority
      (CallFunction|BindEnergyComponentListeners)
      ;;[1 1 1] Server doesn't have the right ammo count initially which we need for energy init, so wait until it gets replicated
      (CallFunction|RemovePreviousCooldown)
      (CallFunction|SpawnHeldShieldActor)
      (EventDispatchers|BindEventtoOnLocalAmmoChanged (delegate Custom|OnAmmoChanged))
      (Utilities|FlowControl|Delay 0.2)
      (CallFunction|OnAmmoChanged 0))))

(event Custom|OnAmmoChanged
  (params (LocalCount :type Integer) (LocalRemaining :type Integer))
  ;[1 1 1] Initialization
  ;;[1 1 1] Server doesn't have the right ammo count initially which we need for energy init, so wait until it gets replicated
  (Utilities|FlowControl|DoOnce false
    (:Completed
      (CallFunction|InitializeHeldShieldEnergy (Variables|Default|GetShieldBGA))
      (EventDispatchers|UnbindEventfromOnLocalAmmoChanged (delegate Custom|OnAmmoChanged)))))

(event Custom|ShieldFormTagListeners
  (bind _returnvalue (Game|GetInstigator))
  (bind _shieldactivetag (Variables|Default|GetShieldActiveTag))
  (Networking|SwitchHasAuthority
    (:Authority
      (Ability|Tasks|WaitGameplayTagAddToActor _returnvalue _shieldactivetag
        (:then
          (Ability|Tasks|WaitGameplayTagRemoveFromActor _returnvalue _shieldactivetag
            (:Removed
              (Variables|Default|SetbActiveShieldTag false))))
        (:Added
          (Variables|Default|SetbActiveShieldTag true))))))

(event Custom|GetSavedEnergy
  ;[1 1 1] Send gameplay event letting the player know a shield is initializing. If it was picked up from a deployed shield, the player will have an ability containing the cached energy amount waiting and will set the correct ammo
  (Ability|SendGameplayEventtoActor (Game|GetInstigator) (Variables|Default|GetNewShieldInitializedEvent) (Utilities|Struct|MakeGameplayEventData "(TagName=\"\")" self)))

(event Custom|VehicleTagListeners
  (Ability|Tasks|WaitGameplayTagAddToActor (Game|GetInstigator) "(TagName=\"Gameplay.Status.InVehicle\")"
    (:Added
      (CallFunction|OnEnteredVehicle))))

(event Custom|InitializeActionPrompts)

(event Custom|OnEnteredVehicle
  (Variables|Default|SetCanActivate false)
  (Utilities|Time|SetTimerbyEvent (delegate Custom|VehicleTempPreventActivateTimerEnded) 0.1))

(event Custom|OnShieldFormTagChanged
  (params (Added :type Boolean))
  (if Added
    ;[1 1 1] Shield activated
    (CharacterAnimation|SetWeaponPawnAnimsetOverride "/ShielderGameplay/Animation/AnimSet/BubbleShield_Active_AnimSet.BubbleShield_Active_AnimSet")
    (CharacterAnimation|SetWeaponCoreAnimOverride "GoingCommando")
    (else
      (CharacterAnimation|SetWeaponPawnAnimsetOverride "/ShielderGameplay/Animation/AnimSet/BubbleShield_NonActive_AnimSet.BubbleShield_NonActive_AnimSet")
      (CharacterAnimation|ResetWeaponCoreAnimOverride))))

(event Custom|VehicleTempPreventActivateTimerEnded
  (Variables|Default|SetCanActivate true))

(event FortWeapon|EventOnEquippedAndAttached
  (|CallParentFunction)
  (CallFunction|ToggleActionPrompts false))

(event FortWeapon|EventOnUnEquippedAndDetached
  (bind _shieldbga (Variables|Default|GetShieldBGA))
  (|CallParentFunction)
  ;[1 1 1] Hide and disable collision on active shield on unequip in case it didn't get cleaned up properly
  (Utilities|IsValid _shieldbga
    (:Is%20Valid
      (Mode|SetShieldVisible _shieldbga false))))

(event EventShowReticle
  ;[1 1 1] Overriding ShowReticle because the parent doesn't work
  (bind _asfort_player_controller_zone (Utilities|Casting|CastToFortPlayerControllerZone (Game|GetInstigatorController)))
  (HUD|SetHUDElementVisibilitywithContext _asfort_player_controller_zone (Variables|Scope|GetReticleHUDElementTags) false "None" ""))

(event Custom|StartThrowTagListeners
  (bind _returnvalue (Game|GetInstigator))
  ;[1 1 1] Toggle reticle when throwing
  (Ability|Tasks|WaitGameplayTagAddToActor _returnvalue (Variables|Default|GetCookingThrowTag)
    (:then
      (Ability|Tasks|WaitGameplayTagRemoveFromActor _returnvalue (Variables|Default|GetCookingThrowTag)
        (:Removed
          (CallFunction|ModifyReticle false))))
    (:Added
      (CallFunction|ModifyReticle true))))

(event Custom|ModifyReticle
  (params (IsThrowing :type Boolean)))

(event Custom|EnergyDeploySequence
  (params (Deploy :type Boolean))
  (if Deploy
    (bind (_direction _deployparameter) (Timeline|TL_EnergyDeploy PlayFromStart)
      (:Update
        (Rendering|Material|SetScalarParameterValue (Variables|Default|GetEnergyDynamicMat) "Deploy" _deployparameter :!class /Script/Engine.MaterialInstanceDynamic)))
    (else
      (Timeline|TL_EnergyDeploy ReverseFromEnd))))

(event Custom|TouchPressTrigger%20
  (params (Weapon :type Fort%20Weapon%20Object%20Reference))
  ;[1 1 1] Mobile logic for toggle ability
  (if (Variables|Default|GetMobileTapAnywhereInteract)
    (Ability|SendGameplayEventtoActor (Game|GetInstigator) "(TagName=\"Ability.Shielder.IsShielding\")")
    (Variables|Default|SetMobileTapAnywhereInteract false)
    (else
      (Variables|Default|SetMobileTapAnywhereInteract true))))

(event Custom|OnShieldWeaponDestroyed
  (params (DestroyedActor :type Actor%20Object%20Reference))
  (Actor|DestroyActor (Variables|Default|GetShieldBGA))
  (CallFunction|RemovePreviousCooldown))

(event Custom|ListenToActorEvents
  ;[1 1 1] Mobile logic for toggle ability
  (Ability|Async|WaitGameplayEventToActor (Game|GetInstigator) "(TagName=\"Ability.Shielder.Event.CancelIsShielding\")"
    (:EventReceived
      (Variables|Default|SetMobileTapAnywhereInteract false))))

(fn BindEnergyComponentListeners
  (bind _b_shield_energytracker_component (Variables|Default|GetB_Shield_EnergyTracker_Component))
  (Default|BindEventtoOnEnergyDepleted _b_shield_energytracker_component (EventDispatchers|CreateEvent :!event OnEnergyDepleted))
  (Default|BindEventtoOnEnergyFull _b_shield_energytracker_component (EventDispatchers|CreateEvent :!event OnEnergyFull))
  (Default|BindEventtoOnEnergyUpdated _b_shield_energytracker_component (EventDispatchers|CreateEvent :!event OnEnergyUpdated)))

(fn CreateEnergyWidget
  (bind _returnvalue (Game|GetInstigator))
  (if (Pawn|IsLocallyControlled _returnvalue)
    (bind _returnvalue_2 (UserInterface|CreateWidget "/ShielderGameplay/Gameplay/ProtoShielderWidget.ProtoShielderWidget_C" (Fortnite|GetFortPlayerControllerfromActor _returnvalue) 1.0))
    (bind _output_get (Variables|Default|SetWidget _returnvalue_2))
    (UserInterface|Viewport|AddtoViewport _output_get 0)))

(fn GetOutOfAmmoTextOverride
  (return "NSLOCTEXT(\"[5532AD7E4B0D5DAAC29B539E2B8ED1B4]\", \"EFDC9EAF40E862010669A89E0AC88020\", \"NOT READY\")"))

(fn InitializeHeldShieldEnergy
  (params (HeldShield :type BGA%20Shielder%20Held%20Object%20Reference))
  (bind _shieldenergycomponent (Variables|/ShielderGameplay/Gameplay/BGA_Shielder_Parent.BGA_Shielder_Parent_C|GetShieldEnergyComponent HeldShield))
  (Default|BindEventtoOnEnergyUpdated _shieldenergycomponent (EventDispatchers|CreateEvent :!event OnEnergyUpdated))
  (Class|BShieldEnergyTrackerComponent|Initialize _shieldenergycomponent self true)
  (return))

(fn OnEnergyDepleted)

(fn OnEnergyFull)

(fn OnEnergyUpdated
  (params (Energy :type Float%20%28double-precision%29))
  (Variables|Default|SetEnergyPercent Energy))

(fn OnRep_bActiveShieldTag
  (CallFunction|OnShieldFormTagChanged (Variables|Default|GetbActiveShieldTag)))

(fn OnRep_CurrentMode
  (bind _currentmode (Variables|Default|GetCurrentMode))
  (bind _shieldbga (Variables|Default|GetShieldBGA))
  (Development|PrintString (Utilities|String|EnumtoString _currentmode))
  (Networking|SwitchHasAuthority
    (:Authority
      (Mode|SetShieldMode _shieldbga _currentmode))
    (:Remote
      _))
  (if (not (Variables|/ShielderGameplay/Gameplay/BGA_Shielder_Held.BGA_Shielder_Held_C|GetbHasPendingPropertiesToApply _shieldbga))
    (CallFunction|SetWeaponActorVisibility (Utilities|Enum|Equal(Enum) (Variables|Default|GetCurrentMode) "NewEnumerator0"))))

(fn OnRep_EnergyPercent
  (Class|ProtoShielderWidget|UpdateFill (Variables|Default|GetWidget) (Variables|Default|GetEnergyPercent)))

(fn OnRep_IsDeployed)

(fn RemoveEnergyWidget
  (bind _widget (Variables|Default|GetWidget)
    (:then
      (Widget|RemovefromParent _widget))
    (:else)))

(fn RemovePreviousCooldown
  (GameplayEffects|RemoveActiveEffectswithGrantedTags (Ability|GetFortAbilitySystemComponent (Game|GetInstigator)) (Variables|Default|GetCooldownTag)))

(fn SetMode
  (params (NewMode :type E_ShieldWeaponMode%20Enum :default "NewEnumerator0"))
  (Variables|Default|SetCurrentMode "NewEnumerator0"))

(fn SetWeaponActorVisibility
  (params (IsVisible :type Boolean))
  (Development|PrintString (Utilities|String|Append "Shield weapon actor visible: " (Utilities|String|ToString(Boolean) IsVisible)) true true "(R=1.000000,G=0.239254,B=0.000000,A=1.000000)")
  (Rendering|SetVisibility (Variables|Visual|GetWeaponMesh) IsVisible false))

(fn SpawnHeldShieldActor
  (bind _returnvalue (Game|GetInstigator))
  (bind _returnvalue_2 (Building|SpawnBuildingGameplayActor (Variables|Default|GetShieldBGAClass) (Transformation|GetActorTransform _returnvalue) _returnvalue true))
  (bind _asbga_shielder_held (Utilities|Casting|CastToBGA_Shielder_Held _returnvalue_2))
  (bind _output_get (Variables|Default|SetShieldBGA _asbga_shielder_held))
  (Variables|/ShielderGameplay/Gameplay/BGA_Shielder_Held.BGA_Shielder_Held_C|SetWeaponActorRef self _output_get)
  (Actor|SetOwner _output_get _returnvalue)
  (bind _asfort_pawn (Utilities|Casting|CastToFortPawn (Game|GetInstigator)))
  (Transformation|AttachActorToComponent _output_get (Variables|/Script/Engine.Character|GetMesh _asfort_pawn) (Variables|AttachmentOverride|GetShieldSocketOverride) "KeepWorld" "KeepWorld" "KeepWorld" true))

(fn StartDepletingEnergy
  (Networking|SwitchHasAuthority
    (:Authority
      (Class|BShieldEnergyTrackerComponent|StartDepleting (Variables|/ShielderGameplay/Gameplay/BGA_Shielder_Parent.BGA_Shielder_Parent_C|GetShieldEnergyComponent (Variables|Default|GetShieldBGA))))))

(fn StopDepletingEnergy
  (Networking|SwitchHasAuthority
    (:Authority
      (Class|BShieldEnergyTrackerComponent|StopDepleting (Variables|/ShielderGameplay/Gameplay/BGA_Shielder_Parent.BGA_Shielder_Parent_C|GetShieldEnergyComponent (Variables|Default|GetShieldBGA)) false false))))

(fn ToggleActionPrompts
  (params (HasDetachedShield :type Boolean))
  (bind _returnvalue (not HasDetachedShield))
  (WeaponHUDData|SetHudKeyActionVisibilitybyID "UseAction" _returnvalue)
  (WeaponHUDData|SetHudKeyActionVisibilitybyID "PlaceAction" _returnvalue)
  (WeaponHUDData|SetHudKeyActionVisibilitybyID "PickUpAction" HasDetachedShield))

(fn ConstructionScript
  (|CallParentFunction))
