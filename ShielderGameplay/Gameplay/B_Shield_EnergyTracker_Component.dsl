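; Shield energy tracker component that drains and recharges a Shielder weapon's energy, stored as loaded ammo.
; A looping timer feeds ModifyEnergy, which updates weapon ammo or a cached count, syncs the owner's item entry and broadcasts energy events.
; Stopping depletion can start a delayed recharge, with a SetByCaller-duration cooldown GameplayEffect after full depletion.
; Infinite loaded energy skips draining and is rechecked on team or class changes.
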(event EventBeginPlay
  (CallFunction|InitializeValuesfromData)
  (CallFunction|CheckHasInfiniteEnergy)
  ;[0.7552 0.7396 0.2641 0.5] Keep track of team/class in case the owning player changes to a team/class with a different setting for infinite loaded energy
  (Minigame|BindEventtoOnPlayerClassApplied (Minigame|GetMinigame (Components|GetOwner) :!class /Script/FortniteGame.MinigameBlueprintLibrary) (delegate Custom|OnPlayerClassApplied_Event)) ; There's no delegate for this from the controller
  (bind _instigatorcontroller (Variables|Default|GetInstigatorController)
    (:then
      (FortPlayerControllerAthena|BindEventtoOnLocalPlayerChangedTeams _instigatorcontroller (delegate Custom|OnLocalPlayerChangedTeams_Event)))
    (:else)))

(event Custom|OnLocalPlayerChangedTeams_Event
  ;[0.7552 0.7396 0.2641 0.5] Keep track of team/class in case the owning player changes to a team/class with a different setting for infinite loaded energy
  (CallFunction|CheckHasInfiniteEnergy))

(event Custom|OnPlayerClassApplied_Event
  (params (PlayerState :type Fort%20Player%20State%20Object%20Reference) (ClassSlotIndex :type Byte))
  ;[0.7552 0.7396 0.2641 0.5] Keep track of team/class in case the owning player changes to a team/class with a different setting for infinite loaded energy
  (if (== (Fortnite|GetControllerfromActor PlayerState) (Variables|Default|GetInstigatorController))
    (CallFunction|CheckHasInfiniteEnergy)))

(event Custom|ToggleShielderDepleteEnergy
  (params (ShouldDeplete :type Boolean))
  (Variables|Default|SetDebugSkipDepleteEnergy (not ShouldDeplete)))

(fn CheckHasInfiniteEnergy
  (Variables|Default|GetInstigatorController
    (:then
      _)
    (:else
      (bind _asfort_player_controller_athena (Utilities|Casting|CastToFortPlayerControllerAthena (Game|GetInstigatorController (Components|GetOwner)))
        (:then
          (Variables|Default|SetInstigatorController _asfort_player_controller_athena))
        (:CastFailed
          (Variables|Default|SetbHasInfiniteEnergy (ScalableFloat|GetValueasBoolwithContextParam (Components|GetOwner) (Variables|/Script/FortniteGame.FortExperienceSettingsComponent|GetInfiniteLoadedEnergy (Fortnite|Athena|GameState|GetExperienceSettingsComponentfromGameState))))))))
  (Variables|Default|SetbHasInfiniteEnergy (Class|FortPlayerControllerAthena|HasInfiniteLoadedEnergy (Variables|Default|GetInstigatorController) 0)))

(fn CopyProperties
  (params (OtherComponent :type B%20Shield%20Energy%20Tracker%20Component%20Object%20Reference))
  (Variables|Default|SetEnergyDepleteRate (Variables|Default|GetEnergyDepleteRate OtherComponent))
  (Variables|Default|SetEnergyRechargeRate (Variables|Default|GetEnergyRechargeRate OtherComponent))
  (Variables|Default|SetEnergyUpdateInterval (Variables|Default|GetEnergyUpdateInterval OtherComponent))
  (Variables|Default|SetCurrentEnergy (Variables|Default|GetCurrentEnergy OtherComponent)))

(fn GetOwnerAbilitySystemComponent
  (bind _asfort_player_pawn (Utilities|Casting|CastToFortPlayerPawn (Variables|/Script/Engine.Actor|GetInstigator (Components|GetOwner))))
  (return (Ability|GetFortAbilitySystemComponent _asfort_player_pawn)))

(fn HandleEnergyDepleted
  (Default|CallOnEnergyDepleted))

(fn HandleEnergyFull
  (Default|CallOnEnergyFull))

(fn Initialize
  (params (ShieldWeapon :type Fort%20Weapon%20Object%20Reference) (StartRecharging :type Boolean))
  (bind _currentenergy (Variables|Default|GetCurrentEnergy))
  (bind _output_get (Variables|Default|SetShield%20Weapon ShieldWeapon))
  (bind _output_get_2 (Variables|Default|SetClipSize (FortWeapon|GetBulletsPerClip _output_get)))
  (Variables|Default|SetCurrentEnergy (/ (Math|Conversions|ToFloat(Integer) (FortWeapon|GetRemainingAmmo _output_get)) (Math|Conversions|ToFloat(Integer) _output_get_2)))
  (if (< _currentenergy 1.0)
    (if (Variables|GetStartRecharging)
      (Default|StopEnergyTimer)
      (CallFunction|SetRechargeTimer (== _currentenergy 0))
      (else
        (Utilities|Time|ClearandInvalidateTimerbyHandle (Variables|Default|GetRechargeDelayTimer))))))

(fn InitializeValuesFromData
  (Variables|Default|SetEnergyUpdateInterval (ScalableFloat|GetValueatLevel (Variables|Default|GetRow_UpdateInterval)))
  (Variables|Default|SetRechargeDelayAfterFullDeplete (ScalableFloat|GetValueatLevel (Variables|Default|GetRow_RechargeDelayAfterFullDeplete)))
  (Variables|Default|SetDefaultRechargeDelay (ScalableFloat|GetValueatLevel (Variables|Default|GetRow_DefaultRechargeDelay))))

(fn ModifyEnergy
  (params (ModAmount :type Float%20%28double-precision%29))
  (bind _shield_weapon (Variables|Default|GetShield%20Weapon))
  (bind _modamount (Variables|GetModAmount))
  (if (not (Variables|Default|GetDebugSkipDepleteEnergy))
    (Utilities|IsValid (Variables|Default|GetShield%20Weapon)
      (:Is%20Valid
        (FortWeapon|ModAmmo _shield_weapon (Math|Float|Truncate _modamount))
        (Variables|Default|SetLastAmmoCount (FortWeapon|GetMagazineAmmoCount _shield_weapon)))
      (:Is%20Not%20Valid
        (Variables|Default|SetLastAmmoCount (Math|Integer|Clamp(Integer) (+ (Variables|Default|GetLastAmmoCount) (Math|Float|Truncate _modamount)) 0 (Variables|Default|GetClipSize)))))
    (CallFunction|TryUpdateItemEntryAmmo (Variables|Default|GetLastAmmoCount))
    (bind _output_get (Variables|Default|SetCurrentEnergy (/ (Variables|Default|GetLastAmmoCount) (Variables|Default|GetClipSize))))
    (Default|CallOnEnergyUpdated _output_get)
    (if (Variables|Default|GetIsDepleting)
      (if (<= (Variables|Default|GetCurrentEnergy) 0)
        (bind _output_get_2 (Variables|Default|SetCurrentEnergy 0.0))
        (Default|CallOnEnergyUpdated _output_get_2)
        (CallFunction|StopDepleting true))
      (else
        (if (>= (Variables|Default|GetCurrentEnergy) 1.0)
          (bind _output_get_3 (Variables|Default|SetCurrentEnergy 1.0))
          (Default|CallOnEnergyUpdated _output_get_3)
          (Default|StopEnergyTimer)
          (CallFunction|HandleEnergyFull))))))

(fn Reciprocal
  (params (Value :type Float%20%28double-precision%29))
  (return (/ 1.0 Value)))

(fn SetRechargeTimer
  (params (bFullyDepleted :type Boolean))
  (bind _defaultrechargedelay (Variables|Default|GetDefaultRechargeDelay))
  (bind _rechargedelayafterfulldeplete (Variables|Default|GetRechargeDelayAfterFullDeplete))
  (bind _returnvalue (Math|Float|SelectFloat _rechargedelayafterfulldeplete _defaultrechargedelay bFullyDepleted))
  (bind _returnvalue_2 (CallFunction|GetOwnerAbilitySystemComponent))
  (Utilities|Time|ClearandInvalidateTimerbyHandle (Variables|Default|GetRechargeDelayTimer))
  (bind _returnvalue_3 (Utilities|Time|SetTimerbyEvent (EventDispatchers|CreateEvent :!event StartRecharging) _returnvalue))
  (Variables|Default|SetRechargeDelayTimer _returnvalue_3)
  (Development|PrintString (Utilities|String|ToString(Float) _returnvalue) true true "(R=1.000000,G=0.117299,B=0.000000,A=1.000000)")
  (if (Variables|GetbFullyDepleted)
    (Utilities|IsValid _returnvalue_2
      (:Is%20Valid
        (bind _returnvalue_4 (Ability|GameplayEffect|AssignTagSetbyCallerMagnitude (GameplayEffects|MakeOutgoingSpec :self _returnvalue_2 :GameplayEffectClass (Variables|Default|GetGE_Cooldown) :Level 0.0) "(TagName=\"SetByCaller.Duration\")" _returnvalue))
        (GameplayEffects|ApplyGameplayEffectSpecToSelf _returnvalue_2 _returnvalue_4)))))

(fn StartDepleting
  (Utilities|Time|ClearandInvalidateTimerbyHandle (Variables|Default|GetRechargeDelayTimer))
  (Variables|Default|SetIsDepleting true)
  (CallFunction|StartEnergyTimer))

(fn StartEnergyTimer
  (bind _returnvalue (Utilities|Time|SetTimerbyFunctionName 0 "UpdateEnergy" (Variables|Default|GetEnergyUpdateInterval) true))
  (Variables|Default|SetEnergyTimer _returnvalue))

(fn StartRecharging
  (Variables|Default|SetIsDepleting false)
  (Utilities|FlowControl|Sequence
    (:then_0
      (if (== (Variables|Default|GetCurrentEnergy) 0)
        (CallFunction|ModifyEnergy (ScalableFloat|GetValueatLevel (Variables|Default|GetMinEnergyOnRechargeStart)))))
    (:then_1
      (CallFunction|StartEnergyTimer))))

(fn StopDepleting
  (params (bFullyDepleted :type Boolean) (DontStartRechargeTimer :type Boolean))
  (if (Variables|Default|GetIsDepleting)
    (Variables|Default|SetIsDepleting false)
    (Default|StopEnergyTimer)
    (if (not DontStartRechargeTimer)
      (CallFunction|SetRechargeTimer bFullyDepleted))
    (if (Variables|GetbFullyDepleted)
      (CallFunction|HandleEnergyDepleted))))

(fn StopEnergyTimer
  (Utilities|Time|ClearandInvalidateTimerbyHandle (Variables|Default|GetEnergyTimer)))

(fn TryUpdateItemEntryAmmo
  (params (NewAmmoAmount :type Integer))
  (bind _returnvalue (Components|GetOwner))
  (bind _returnvalue_2 (Actor|GetComponentbyClass _returnvalue "/Script/FortniteGame.FortItemEntryComponent"))
  (Utilities|IsValid _returnvalue_2
    (:Is%20Valid
      (Class|FortItemEntryComponent|TrySetLoadedAmmo _returnvalue_2 NewAmmoAmount))))

(fn UpdateEnergy
  (if (Variables|Default|GetIsDepleting)
    (if (Variables|Default|GetbHasInfiniteEnergy)
      (CallFunction|ModifyEnergy 0.0) ; Still call ModifyEnergy so everything is properly updated
      (else
        (CallFunction|ModifyEnergy (* (Variables|Default|GetEnergyDepleteRate) -1.0))))
    (else
      (CallFunction|ModifyEnergy (Variables|Default|GetEnergyRechargeRate)))))
