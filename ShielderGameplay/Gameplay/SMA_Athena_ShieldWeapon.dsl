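; Deployable shield weapon actor that follows its owning pawn's control pitch and yaw while attached, or lets same-team players reclaim it as an item in one mode.
; Replicated mode changes swap held/placed mesh, offset and target scale.
; A timeline animates scale and dynamic-material opacity, reversing on removal.
; Interaction, death or energy depletion leads to removal; the server then destroys it, spawning a pickup if detached and depleted.
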(event EventBeginPlay
  (bind _staticmesh (Variables|Default|GetStaticMesh))
  (bind _asfort_pawn (Utilities|Casting|CastToFortPawn (Actor|GetOwner))
    (:then
      (Variables|Default|SetOwningFortPawn _asfort_pawn)
      (bind _returnvalue (Rendering|Material|CreateDynamicMaterialInstance (Variables|Default|GetTargetShieldMaterial) "None" "None" :!class /Script/Engine.KismetMaterialLibrary))
      (bind _output_get (Variables|Default|SetDynamicShieldMaterial _returnvalue))
      (Rendering|Material|SetScalarParameterValue _output_get "Opacity" 0.0 :!class /Script/Engine.MaterialInstanceDynamic)
      (Transformation|SetWorldScale3D _staticmesh (Variables|Default|GetStartingScale))
      (Rendering|Material|SetMaterial _staticmesh 0 _output_get))
    (:CastFailed
      (Actor|DestroyActor))))

(event Custom|ScaleMeshSequence
  (params (In :type Boolean))
  (Components|Timeline|SetPlayRate (Variables|SMA_Athena_ShieldWeapon|GetFadeinTimeline) (Utilities|Reciprocal (Math|Float|SelectFloat (Variables|Default|GetFadeInSpeed) (Variables|Default|GetFadeOutSpeed) In)))
  (if In
    (bind (_direction _float) (Timeline|FadeinTimeline PlayFromStart)
      (:Update
        (Rendering|Material|SetScalarParameterValue (Variables|Default|GetDynamicShieldMaterial) "Opacity" _float :!class /Script/Engine.MaterialInstanceDynamic)
        (Transformation|SetWorldScale3D (Variables|Default|GetStaticMesh) (Math|Vector|Lerp(Vector) (Variables|Default|GetStartingScale) (Variables|Default|GetTargetScale) _float)))
      (:Finished
        (if (not In)
          (Networking|SwitchHasAuthority
            (:Authority
              (CallFunction|TrySpawnPickupifDetached)
              (Actor|DestroyActor))))))
    (else
      (Timeline|FadeinTimeline ReverseFromEnd))))

(event Interaction|EventBlueprintOnInteract
  (params (InteractingPawn :type Fort%20Pawn%20Object%20Reference) (InteractionBeingAttempted :type EInteractionBeingAttempted%20Enum :default "FirstInteraction"))
  (CallFunction|GrantItemtoNewOwner InteractingPawn)
  (CallFunction|MulticastRemoveShield))

(event EventOnDeathServer
  (params (Damage :type Float%20%28single-precision%29) (DamageTags :type Gameplay%20Tag%20Container%20Structure%20%28by%20ref%29) (Momentum :type Vector :default "0, 0, 0") (HitInfo :type Hit%20Result%20Structure%20%28by%20ref%29) (InstigatedBy :type Controller%20Object%20Reference) (DamageCauser :type Actor%20Object%20Reference) (EffectContext :type Gameplay%20Effect%20Context%20Handle%20Structure))
  (CallFunction|MulticastRemoveShield))

(event Custom|Multicast_RemoveShield
  (CallFunction|RemoveShield))

(fn AttachToNewOwner
  (params (NewOwner :type Fort%20Pawn%20Object%20Reference))
  (Variables|Default|SetOwningFortPawn NewOwner)
  (Variables|Default|SetCurrentMode "NewEnumerator2")
  (Variables|Default|SetIsAttached true)
  (Transformation|AttachActorToActor NewOwner "None" "SnapToTarget" "SnapToTarget"))

(fn BindEnergyComponentListeners
  (Default|BindEventtoOnEnergyDepleted (Variables|Default|GetB_Shield_EnergyTracker_Component) (EventDispatchers|CreateEvent :!event OnEnergyDepleted)))

(fn BlueprintCanInteract
  (params (InteractingPawn :type Fort%20Pawn%20Object%20Reference) (InteractionBeingAttempted :type EInteractionBeingAttempted%20Enum :default "FirstInteraction") (InteractionType :type TInteractionType%20Enum :default "IT_NoInteraction"))
  (return (and (Utilities|Enum|Equal(Enum) (Variables|Default|GetCurrentMode) "NewEnumerator4") (Team|OnSameTeam InteractingPawn (Variables|Default|GetOwningFortPawn)))))

(fn FindItemPickupSpawnLocation
  (bind _outhit (Collision|LineTraceByChannel (Utilities|Operators|Add (Transformation|GetActorLocation) (Math|Vector|MakeVector 0.0 0.0 (Variables|Default|GetItemPickupOffset))) (Utilities|Operators|Subtract (Transformation|GetActorLocation) (Math|Vector|MakeVector 0.0 0.0 (Variables|Default|GetItemPickupOffset))) "TraceTypeQuery10"))
  (if (Collision|BreakHitResult _outhit)
    (return (.Location (Collision|BreakHitResult _outhit)))
    (else
      (return (.TraceStart (Collision|BreakHitResult _outhit))))))

(fn GrantItemToNewOwner
  (params (NewOwner :type Fort%20Pawn%20Object%20Reference))
  (bind _asfort_player_pawn (Utilities|Casting|CastToFortPlayerPawn NewOwner))
  (Inventory|GiveItemtoInventoryOwner (Fortnite|GetFortPlayerControllerfromActor _asfort_player_pawn) (Variables|Default|GetShieldItemDefinition)))

(fn OnActorTick
  (if (Variables|Default|GetIsAttached)
    (bind _owningfortpawn (Variables|Default|GetOwningFortPawn)
      (:then
        (bind (_returnvalue_roll _returnvalue_pitch _returnvalue_yaw) (Math|Rotator|BreakRotator (Pawn|GetControlRotation _owningfortpawn :!class /Script/Engine.Pawn)))
        (Transformation|SetActorRotation 0 (Math|Rotator|MakeRotator 0.0 _returnvalue_pitch _returnvalue_yaw) false))
      (:else))))

(fn OnEnergyDepleted
  (Gameplay|Die))

(fn OnRep_CurrentMode
  (CallFunction|SetScalebyMode)
  (CallFunction|SetMeshbyMode)
  (CallFunction|SetRelativeOffsetbyMode))

(fn RemoveShield
  (Variables|Default|SetStartingScale "0, 0, 0")
  (CallFunction|ScaleMeshSequence false))

(fn SetMeshByMode
  (Components|StaticMesh|SetStaticMesh (Variables|Default|GetStaticMesh) (select (Utilities|Array|ContainsItem (Variables|Default|GetHeldModes) (Variables|Default|GetCurrentMode)) (Variables|Default|GetHeldMesh) (Variables|Default|GetPlacedMesh))))

(fn SetRelativeOffsetByMode
  (bind _currentmode (Variables|Default|GetCurrentMode))
  (bind _modeoffsetmap (Variables|Default|GetModeOffsetMap))
  (bind (_value _returnvalue) (Utilities|Map|Find _modeoffsetmap _currentmode))
  (if _returnvalue
    (Transformation|SetActorRelativeLocation (Math|Vector|SelectVector _value "0, 0, 0" _returnvalue))))

(fn SetScaleByMode
  (bind _currentmode (Variables|Default|GetCurrentMode))
  (bind _modetargetscalemap (Variables|Default|GetModeTargetScaleMap))
  (bind (_value _returnvalue) (Utilities|Map|Find _modetargetscalemap _currentmode))
  (if _returnvalue
    (Variables|Default|SetTargetScale _value)
    (Variables|Default|SetStartingScale (Transformation|GetWorldScale (Variables|Default|GetStaticMesh)))
    (CallFunction|ScaleMeshSequence true)))

(fn TrySpawnPickupIfDetached
  (if (and (Utilities|Enum|Equal(Enum) (Variables|Default|GetCurrentMode) "NewEnumerator4") (<= (Variables|/ShielderGameplay/Gameplay/B_Shield_EnergyTracker_Component.B_Shield_EnergyTracker_Component_C|GetCurrentEnergy (Variables|Default|GetB_Shield_EnergyTracker_Component)) 0))
    (bind _location (CallFunction|FindItemPickupSpawnLocation))
    (Inventory|SpawnItemVariantPickupinWorld (Utilities|Struct|MakeSpawnItemVariantParams (Variables|Default|GetShieldItemDefinition) "(GameplayTags=)" 1 _location "0.000000,0.000000,1.000000" 0 true true true -1 "Tossed" "ItemSpawner" 0 0 false))))

(fn ConstructionScript)
